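; Explosive held object (TnTina) that players carry and throw, disabling water interaction while held and tracking the last damage instigator.
; Damage or Curie fire starts delayed fuses with sound and a replicated on-fire cue, leading to ExplodeAndDie.
; With Curie active and out of water, that spawns fire and random cone-spread fireballs.
; OnRep_bShouldDie plays land or water explosion cues, drops the object, triggers the explosion ability by gameplay event and destroys the actor.
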(event OnHeldObjectPickedUp(FortHeldObject)
  ;[0.1686 0.7886 1] Turn off the water component on pickup and re-enable it on drop/throw/place. This is to prevent the TnTina from breaking the offset when picked up in water
  (Components|Activation|Deactivate (Variables|Default|GetFortWaterInteraction))
  (bind _returnvalue (Fortnite|HeldObjects|GetHeldObjectsOwningPlayer self))
  (Utilities|IsValid _returnvalue
    (:Is%20Valid
      (EventDispatchers|AssignOnHeldObjectThrown (Variables|Default|GetFortHeldObject) (delegate Custom|Set%20Damage%20Instigator%20on%20Throw)))))

(event OnHeldObjectDropped(FortHeldObject)
  ;[0.1686 0.7886 1] Turn off the water component on pickup and re-enable it on drop/throw/place. This is to prevent the TnTina from breaking the offset when picked up in water
  (Components|Activation|Activate (Variables|Default|GetFortWaterInteraction) false))

(event OnHeldObjectPlaced(FortHeldObject)
  ;[0.1686 0.7886 1] Turn off the water component on pickup and re-enable it on drop/throw/place. This is to prevent the TnTina from breaking the offset when picked up in water
  (Components|Activation|Activate (Variables|Default|GetFortWaterInteraction) false))

(event OnHeldObjectThrownLanded(FortHeldObject)
  ;[0.1686 0.7886 1] Turn off the water component on pickup and re-enable it on drop/throw/place. This is to prevent the TnTina from breaking the offset when picked up in water
  (Components|Activation|Activate (Variables|Default|GetFortWaterInteraction) false))

(event Custom|Set%20Damage%20Instigator%20on%20Throw
  ;[0.1686 0.7886 1] Turn off the water component on pickup and re-enable it on drop/throw/place. This is to prevent the TnTina from breaking the offset when picked up in water
  (Variables|Default|SetLastDamageInstigator (Fortnite|GetFortPlayerControllerfromActor (stale OnHeldObjectPickedUp(FortHeldObject) _returnvalue))))

(event OnHeldObjectThrown(FortHeldObject)
  ;[0.1686 0.7886 1] Turn off the water component on pickup and re-enable it on drop/throw/place. This is to prevent the TnTina from breaking the offset when picked up in water
  (Components|Activation|Activate (Variables|Default|GetFortWaterInteraction) false))

(event EventOnDeathServer
  (params (Damage :type Float%20%28single-precision%29) (DamageTags :type Gameplay%20Tag%20Container%20Structure%20%28by%20ref%29) (Momentum :type Vector :default "0, 0, 0") (HitInfo :type Hit%20Result%20Structure%20%28by%20ref%29) (InstigatedBy :type Controller%20Object%20Reference) (DamageCauser :type Actor%20Object%20Reference) (EffectContext :type Gameplay%20Effect%20Context%20Handle%20Structure))
  ;[1 1 1] On death
  (EventDispatchers|UnbindEventfromOnHeldObjectThrown (Variables|Default|GetFortHeldObject) (delegate Custom|Set%20Damage%20Instigator%20on%20Throw))
  (Rendering|SetActorHiddenInGame true)
  (bind _returnvalue (MissionHelpers|HasTags DamageTags (GameplayTags|MakeGameplayTagContainerfromArray (Utilities|Array|MakeArray "(TagName=\"Gameplay.Effect.InstantDeath.Silent\")" "(TagName=\"Gameplay.Effect.InstantDeath.StructuralSupport\")" "(TagName=\"Gameplay.Effect.InstantDeath.PersistentUpdate\")" "(TagName=\"Gameplay.Effect.InstantDeath.Environment\")" "(TagName=\"Gameplay.Effect.InstantDeath.Detachment\")" "(TagName=\"Gameplay.Effect.InstantDeath.Despawn\")" "(TagName=\"Gameplay.Effect.InstantDeath.Culled\")" "(TagName=\"Gameplay.Effect.InstantDeath\")")) false false))
  (if (not _returnvalue)
    (CallFunction|ExplosionTrace)
    (Variables|Default|SetbShouldDie true))
  (Development|PrintString (Utilities|String|ToString(Name) (GameplayTags|GetTagName (Utilities|Array|Get(acopy) (GameplayTags|BreakGameplayTagContainer DamageTags)))) true true "(R=0.770487,G=0.000000,B=1.000000,A=1.000000)" 3.0))

(event EventOnDeathPlayEffects
  (params (Damage :type Float%20%28single-precision%29) (DamageTags :type Gameplay%20Tag%20Container%20Structure%20%28by%20ref%29) (Momentum :type Vector :default "0, 0, 0") (HitInfo :type Hit%20Result%20Structure%20%28by%20ref%29) (InstigatedBy :type Fort%20Pawn%20Object%20Reference) (DamageCauser :type Actor%20Object%20Reference) (EffectContext :type Gameplay%20Effect%20Context%20Handle%20Structure))
  ;[1 1 1] On death
  ;;[1 1 1] Play Effects
  (Development|PrintString (Utilities|String|ToString(Name) (GameplayTags|GetTagName (Utilities|Array|Get(acopy) (GameplayTags|BreakGameplayTagContainer DamageTags)))) true true "(R=0.770487,G=0.000000,B=1.000000,A=1.000000)" 3.0))

(event EventBeginPlay
  ;[1 1 1] Begin Play
  (|CallParentFunction)
  (Curie|OnCurieActivated
    (:OnCurieActive
      (Variables|Default|SetCurieActivated true))))

(event Custom|Explode%20and%20Die%20with%20Delay
  (params (LastDamageInstigator :type Fort%20Player%20Controller%20Object%20Reference))
  ;[1 1 1] Death logic when taking damage. When damage is taken, start the fuse and begin the explosion windup.
  (Utilities|IsValid LastDamageInstigator
    (:Is%20Valid
      (Variables|Default|SetLastDamageInstigator LastDamageInstigator))
    (:Is%20Not%20Valid
      _))
  (CallFunction|PlayPreExplodeSound)
  (CallFunction|SetonFire)
  (Utilities|FlowControl|Delay (Math|Random|RandomFloatinRange (ScalableFloat|GetValueatLevel (Variables|Default|GetMin%20Duration%20TnTina%20Fuse)) (ScalableFloat|GetValueatLevel (Variables|Default|GetMax%20Duration%20TnTina%20Fuse))))
  (CallFunction|ExplodeandDie))

(event Custom|ExplodeAndDie
  (bind _capsule (Variables|Default|GetCapsule))
  (bind _returnvalue (Transformation|GetWorldLocation _capsule))
  (bind _capsule_2 (Variables|Default|GetCapsule))
  (bind _returnvalue_2 (Transformation|GetUpVector _capsule_2))
  (bind _capsule_3 (Variables|Default|GetCapsule))
  (bind _returnvalue_3 (Transformation|GetWorldRotation _capsule_3))
  (bind _returnvalue_4 (Transformation|GetWorldLocation _capsule_3))
  (bind _returnvalue_5 (Math|Transform|MakeTransform _returnvalue_4 _returnvalue_3))
  (bind _spawnedprojectileminspeed (Variables|Default|GetSpawnedProjectileMinSpeed))
  (bind _returnvalue_6 (ScalableFloat|GetValueatLevel _spawnedprojectileminspeed))
  (bind _spawnedprojectilemaxspeed (Variables|Default|GetSpawnedProjectileMaxSpeed))
  (bind _returnvalue_7 (ScalableFloat|GetValueatLevel _spawnedprojectilemaxspeed))
  (bind _conehalfangleindegrees (Variables|Default|GetConeHalfAngleInDegrees))
  (bind _returnvalue_8 (ScalableFloat|GetValueatLevel _conehalfangleindegrees))
  (bind _returnvalue_9 (Math|Random|RandomFloatinRange _returnvalue_6 _returnvalue_7))
  (bind _lastdamageinstigator (Variables|Default|GetLastDamageInstigator))
  (bind _returnvalue_10 (Game|Player|GetPlayerPawn _lastdamageinstigator))
  (bind _lastdamageinstigator_2 (Variables|Default|GetLastDamageInstigator))
  (bind _returnvalue_11 (Game|Player|GetPlayerPawn _lastdamageinstigator_2))
  (bind _returnvalue_12 (+ _returnvalue "0, 0,10.000000"))
  (bind _fireball_spawn_vectors (Variables|Default|GetFireball%20Spawn%20Vectors))
  ;[1 1 1] Death logic when taking damage. When damage is taken, start the fuse and begin the explosion windup.
  (Networking|SwitchHasAuthority
    (:Authority
      (Utilities|FlowControl|DoOnce false
        (:Completed
          (Rendering|SetActorHiddenInGame true)
          ;[1 0.4734 0.2717] Is Curie Enabled?
          (if (Variables|Default|GetCurieActivated)
            ;[1 1 1] If In Water, do not spawn Fire. Otherwise spawn fire sphere BGA
            (if (not (Water|IsinWater (Variables|Default|GetFortWaterInteraction)))
              (Utilities|IsValid _lastdamageinstigator_2
                (:Is%20Valid
                  (Utilities|IsValid _returnvalue_11
                    (:Is%20Valid
                      (Building|SpawnBuildingGameplayActor "/Game/Athena/Items/HeldObjects/HeldObjectVersion2/BGAExplosive/BGA_CurieFireMesh_TnTina.BGA_CurieFireMesh_TnTina_C" _returnvalue_5 _returnvalue_11))
                    (:Is%20Not%20Valid
                      (Building|SpawnBuildingGameplayActor "/Game/Athena/Items/HeldObjects/HeldObjectVersion2/BGAExplosive/BGA_CurieFireMesh_TnTina.BGA_CurieFireMesh_TnTina_C" _returnvalue_5 :!label _label_2)))
                  ;[1 1 1] Spawn Projectiles with pseudo random speeds and directions
                  (for _index (range (Math|Float|Truncate (ScalableFloat|GetValueatLevel (Variables|Default|GetFireBallsToSPawn))))
                    (Utilities|IsValid _lastdamageinstigator
                      (:Is%20Valid
                        (Utilities|IsValid _returnvalue_10
                          (:Is%20Valid
                            (Projectile|SpawnProjectile :ProjectileClass "/Game/Athena/Items/HeldObjects/HeldObjectVersion2/BGAExplosive/B_Prj_Athena_TnTina_Fireballs.B_Prj_Athena_TnTina_Fireballs_C" :RequestedBy _returnvalue_10 :SpawnLocation _returnvalue_12 :SpawnRotation "0, 0, 0" :SpawnDirection (Math|Rotator|MakeRotfromX (Math|Random|RandomUnitVectorinConeinDegrees (+ (Utilities|Array|Get(acopy) _fireball_spawn_vectors _index) _returnvalue_2) _returnvalue_8)) :bUseDefaultPhysics false :InitialSpeed _returnvalue_9 :GravityScale 0.6 :ChargePercent 0.0))
                          (:Is%20Not%20Valid
                            (Projectile|SpawnProjectile :ProjectileClass "/Game/Athena/Items/HeldObjects/HeldObjectVersion2/BGAExplosive/B_Prj_Athena_TnTina_Fireballs.B_Prj_Athena_TnTina_Fireballs_C" :SpawnLocation _returnvalue_12 :SpawnRotation "0, 0, 0" :SpawnDirection (Math|Rotator|MakeRotfromX (Math|Random|RandomUnitVectorinConeinDegrees (+ (Utilities|Array|Get(acopy) _fireball_spawn_vectors _index) _returnvalue_2) _returnvalue_8)) :bUseDefaultPhysics false :InitialSpeed _returnvalue_9 :GravityScale 0.6 :ChargePercent 0.0 :!label _label)))
                        (Variables|Default|SetbShouldDie true :!label _label_3))
                      (:Is%20Not%20Valid
                        _))
                    (goto _label)))
                (:Is%20Not%20Valid
                  _))
              (goto _label_2))
            (goto _label_3)
            (else
              (Variables|Default|SetbShouldDie true))))))))

(event EventOnDamageServer
  (params (Damage :type Float%20%28single-precision%29) (DamageTags :type Gameplay%20Tag%20Container%20Structure%20%28by%20ref%29) (Momentum :type Vector :default "0, 0, 0") (HitInfo :type Hit%20Result%20Structure%20%28by%20ref%29) (InstigatedBy :type Controller%20Object%20Reference) (DamageCauser :type Actor%20Object%20Reference) (EffectContext :type Gameplay%20Effect%20Context%20Handle%20Structure))
  (bind (_bblockinghit _binitialoverlap _time _distance _location _impactpoint _normal) (Collision|BreakHitResult HitInfo))
  ;[1 1 1] Death logic when taking damage. When damage is taken, start the fuse and begin the explosion windup.
  (Variables|Default|SetLast%20Hit (CallFunction|MakeLocalHitInfo _location _normal))
  (bind _asfort_player_controller (Utilities|Casting|CastToFortPlayerController InstigatedBy)
    (:then
      (Variables|Default|SetLastDamageInstigator _asfort_player_controller)
      (bind _returnvalue (MissionHelpers|HasTags DamageTags (Variables|Default|GetTC%20Melee%20Damage) true false))
      (if _returnvalue
        (CallFunction|PlayPreExplodeSound)
        (CallFunction|SetonFire)
        (Utilities|FlowControl|Delay (Math|Random|RandomFloatinRange (ScalableFloat|GetValueatLevel (Variables|Default|GetMin%20Duration%20TnTina%20Fuse)) (ScalableFloat|GetValueatLevel (Variables|Default|GetMax%20Duration%20TnTina%20Fuse))))
        (CallFunction|ExplodeandDie)
        (else
          _)))
    (:CastFailed
      (Development|PrintString "No Controller")))
  (Utilities|FlowControl|Sequence
    (:then_0
      (Utilities|FlowControl|DoOnce false
        (:Completed
          (Utilities|Time|SetTimerbyEvent (delegate Custom|ExplodeAndDie) (Math|Random|RandomFloatinRange (ScalableFloat|GetValueatLevel (Variables|Default|GetMin%20Duration%20Damage%20Fuse%20)) (ScalableFloat|GetValueatLevel (Variables|Default|GetMax%20Duration%20Damage%20Fuse)))))))
    (:then_1
      (Utilities|FlowControl|DoOnce false
        (:Completed
          (Variables|Default|SetOn%20Fire%20Hit%20Info (Variables|Default|GetLast%20Hit))
          (CallFunction|PlayPreExplodeSound)
          (Utilities|FlowControl|Delay 1.4)
          (bind _lastdamageinstigator (Variables|Default|GetLastDamageInstigator)
            (:then
              (Utilities|IsValid (Pawn|GetControlledPawn _lastdamageinstigator)
                (:Is%20Valid
                  (GameplayEffects|FortApplyGameplayEffectToTarget :self (Ability|GetFortAbilitySystemComponent (Pawn|GetControlledPawn _lastdamageinstigator)) :GameplayEffectClass "/Game/Athena/Items/HeldObjects/HeldObjectVersion2/BGAExplosive/GE_Athena_ReactiveProps_OnFire_TnTina.GE_Athena_ReactiveProps_OnFire_TnTina_C" :TargetOfEffect (Ability|GetFortAbilitySystemComponent self) :OptionalSourceObject (Pawn|GetControlledPawn _lastdamageinstigator) :Level 0.0))
                (:Is%20Not%20Valid
                  _)))
            (:else
              _))
          (CallFunction|ExplodeandDie))))))

(event Custom|SetOnFire
  ;[1 1 1] Death logic when taking damage. When damage is taken, start the fuse and begin the explosion windup.
  ;;[1 1 1] Set On Fire (only triggered on bounce when travelling fast enough OR via a melee attack)
  (Variables|Default|SetLast%20Hit (CallFunction|MakeLocalHitInfo (Transformation|GetSocketLocation (Variables|Default|GetStaticMesh) "DefaultBurnLocation") (Math|Rotator|GetRotationXVector (Transformation|GetSocketRotation (Variables|Default|GetStaticMesh) "DefaultBurnLocation"))))
  (Utilities|FlowControl|DoOnce false
    (:Completed
      (Variables|Default|SetOn%20Fire%20Hit%20Info (Variables|Default|GetLast%20Hit))
      (CallFunction|PlayPreExplodeSound)
      (Utilities|FlowControl|Delay 1.4)
      (bind _lastdamageinstigator (Variables|Default|GetLastDamageInstigator)
        (:then
          (Utilities|IsValid (Pawn|GetControlledPawn _lastdamageinstigator)
            (:Is%20Valid
              (GameplayEffects|FortApplyGameplayEffectToTarget :self (Ability|GetFortAbilitySystemComponent (Pawn|GetControlledPawn _lastdamageinstigator)) :GameplayEffectClass "/Game/Athena/Items/HeldObjects/HeldObjectVersion2/BGAExplosive/GE_Athena_ReactiveProps_OnFire_TnTina.GE_Athena_ReactiveProps_OnFire_TnTina_C" :TargetOfEffect (Ability|GetFortAbilitySystemComponent self) :OptionalSourceObject (Pawn|GetControlledPawn _lastdamageinstigator) :Level 0.0))
            (:Is%20Not%20Valid
              _)))
        (:else
          _))
      (CallFunction|ExplodeandDie))))

(event Custom|PlayPreExplodeSound
  ;[1 1 1] Pre Explode Sound
  (Audio|PlaySoundatLocation (Variables|Default|GetPreExplodeSound) (Transformation|GetActorLocation)))

(event EventDestroyed
  ;[1 1 1] Failsafe to ensure static mesh is hidden on death
  (Rendering|SetActorHiddenInGame true)
  (Projectile|RemoveGameplayCue%28Looping%29Local self (Variables|Default|GetFireCue) (Ability|GameplayCue|MakeGameplayCueParameters 0.0) false))

(event Custom|ExplosionTrace
  ;[1 1 1] On Explode, sent event to GA_HeldObject_Explode
  (Utilities|FlowControl|DoOnce false
    (:Completed
      (Ability|SendGameplayEventtoActor self "(TagName=\"Abilities.Generic.TriggeredAbility.Activate\")" (Utilities|Struct|MakeGameplayEventData :Instigator (Variables|Default|GetLastDamageInstigator) :ContextHandle "()" :InstigatorTags "(GameplayTags=)" :TargetTags "(GameplayTags=)" :EventMagnitude 0.0 :TargetData "()")))))

(event Curie|EventOnCurieElementInteract
  (params (CurieContainerHandle :type Curie%20Container%20Handle%20Structure) (ElementTag :type Gameplay%20Tag%20Structure) (InteractParamsHandle :type Curie%20Interact%20Params%20Handle%20Structure%20%28by%20ref%29))
  (if (GameplayTags|Equal(GameplayTag) ElementTag "(TagName=\"Curie.Element.Fire\")")
    (bind _instigator (Curie|GetInstigatorInfofromInteractParams InteractParamsHandle))
    (bind _asfort_player_controller (Utilities|Casting|CastToFortPlayerController _instigator)
      (:then
        (Variables|Default|SetLastDamageInstigator _asfort_player_controller))
      (:CastFailed
        _))
    (CallFunction|SetonFire)))

(event Curie|EventOnCurieElementInteractBegun
  (params (CurieContainerHandle :type Curie%20Container%20Handle%20Structure) (ElementTag :type Gameplay%20Tag%20Structure) (InteractParamsHandle :type Curie%20Interact%20Params%20Handle%20Structure%20%28by%20ref%29))
  (if (GameplayTags|Equal(GameplayTag) ElementTag "(TagName=\"Curie.Element.Fire\")")
    (bind _instigator (Curie|GetInstigatorInfofromInteractParams InteractParamsHandle))
    (bind _asfort_player_controller (Utilities|Casting|CastToFortPlayerController _instigator)
      (:then
        (Variables|Default|SetLastDamageInstigator _asfort_player_controller))
      (:CastFailed
        _))
    (CallFunction|SetonFire)))

(fn BlueprintCanInteract
  (params (InteractingPawn :type Fort%20Pawn%20Object%20Reference) (InteractionBeingAttempted :type EInteractionBeingAttempted%20Enum :default "FirstInteraction") (InteractionType :type TInteractionType%20Enum :default "IT_NoInteraction"))
  (bind _returnvalue (Fortnite|HeldObjects|IsHeldObjectBeingHeld self))
  (return (and (or (>= (Health|GetHealth) 1e+01) _returnvalue) (|CallParentFunction InteractingPawn InteractionBeingAttempted))))

(fn GetLastInstigator
  (return (Variables|Default|GetLastDamageInstigator)))

(fn HeldObjectClient/ServerMismatchCorrection
  (params (Actor%20Location :type Vector))
  (bind _fortheldobject (Variables|Default|GetFortHeldObject))
  (bind _returnvalue (HeldObjects|Instigator|GetInstigator _fortheldobject))
  (if (HeldObjects|IsHeld (Variables|Default|GetFortHeldObject))
    (Utilities|IsValid _returnvalue
      (:Is%20Valid
        (return (- Actor%20Location (Math|Vector|SelectVector "0, 0,120.000000" "0, 0,180.000000" (Variables|/Script/Engine.Character|GetbIsCrouched _returnvalue)))))
      (:Is%20Not%20Valid
        _)))
  (return (Variables|GetActor%20Location)))

(fn MakeLocalHitInfo
  (params (Location :type Vector) (Normal :type Vector))
  (return (Collision|MakeHitResult false false 0.0 0.0 (- Location (Transformation|GetActorLocation self)) (- Location (Transformation|GetActorLocation self)) (Math|Vector|UnrotateVector Normal (Transformation|GetActorRotation self)) (Math|Vector|UnrotateVector Normal (Transformation|GetActorRotation self)))))

(fn OnRep_bShouldDie
  (Projectile|RemoveGameplayCue%28Looping%29Local self (Variables|Default|GetFireCue) (Ability|GameplayCue|MakeGameplayCueParameters 0.0))
  (GameplayCue|ExecuteGameplayCue%28Burst%29Local self (select (Water|IsinWater (Variables|Default|GetFortWaterInteraction)) (Variables|Default|GetExplosion%20CUE%20Water) (Variables|Default|GetExplosionCUE)) (Ability|GameplayCue|MakeGameplayCueParameters 0.0) false)
  (if (HeldObjects|IsHeld (Variables|Default|GetFortHeldObject))
    (HeldObjects|Behavior|DropHeldObject (Variables|Default|GetFortHeldObject)))
  (if (Networking|IsServer)
    (CallFunction|ExplosionTrace)
    (Actor|DestroyActor)))

(fn OnRep_OnFireHitInfo
  (bind (_bblockinghit _binitialoverlap _time _distance _location _impactpoint _normal) (Collision|BreakHitResult (Variables|Default|GetOn%20Fire%20Hit%20Info)))
  (bind _self self)
  (Networking|CanExecuteCosmeticEvents
    (:True
      (bind _outcueparameters (GameplayCue|BuildCueParametersfromHitResult _self (Collision|MakeHitResult false false 0.0 0.0 (+ (CallFunction|HeldObjectClient/ServerMismatchCorrection (Transformation|GetActorLocation self)) _location) (+ (CallFunction|HeldObjectClient/ServerMismatchCorrection (Transformation|GetActorLocation self)) _location) (Math|Vector|RotateVector _normal (Transformation|GetActorRotation self)) (Math|Vector|RotateVector _normal (Transformation|GetActorRotation self)))))
      (Projectile|AddGameplayCue%28Looping%29Local _self (Variables|Default|GetFireCue) _outcueparameters))))

(fn ShouldDie
  (params (Damage :type Float%20%28single-precision%29) (EventInstigator :type Controller%20Object%20Reference) (DamageCauser :type Actor%20Object%20Reference) (bIsTest :type Boolean) (EffectContext :type Gameplay%20Effect%20Context%20Handle%20Structure%20%28by%20ref%29))
  ;[1 1 1] This Function get's called on Death. Health Is Zero get's set to true so that the damage event triggers the Death Fx. Should die is set to false so that the actors collision stays intact
  (Variables|Default|SetHealthIsZero true)
  (return (Variables|Default|GetbShouldDie)))

(fn ConstructionScript)
