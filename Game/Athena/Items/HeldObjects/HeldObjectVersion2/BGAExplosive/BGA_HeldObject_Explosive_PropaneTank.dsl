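; Blueprint for the propane tank, an explosive held object actor.
; On a projectile bounce, the authority compares the projectile movement's squared speed against the square of a scalable fuse trigger velocity threshold.
; At or above that threshold it calls Set on Fire.
; BeginPlay and the construction script only defer to the parent class.
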(event EventBeginPlay
  (|CallParentFunction))

(event OnProjectileBounce(FortProjectileMovement)
  (params (ImpactResult :type Hit%20Result%20Structure%20%28by%20ref%29) (ImpactVelocity :type Vector%20%28by%20ref%29 :default "0, 0, 0"))
  (Networking|SwitchHasAuthority
    (:Authority
      (if (>= (Math|Vector|VectorLengthSquared (Variables|/Script/Engine.MovementComponent|GetVelocity (Variables|Default|GetFortProjectileMovement))) (Math|Float|Square (ScalableFloat|GetValueatLevel (Variables|Default|GetFuse%20Trigger%20Velocity%20Threshold))))
        (CallFunction|SetonFire)))))

(fn ConstructionScript
  (|CallParentFunction))
