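; Actor spawner for the explosive propane tank held object, choosing between physics and non-physics versions of the tank.
; On BeginPlay it waits for playlist data, then reads a spawn-physics data row, with a debug override forcing the physics version.
; It sets the class to spawn accordingly, disabling previs mesh scale for the differently scaled non-physics version, before calling parent BeginPlay.
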(event EventBeginPlay
  ;[1 1 1] Check if we should spawn the physics version or not.
  (Class|FortAsyncActionCurrentPlaylistReady|PlaylistDataReadyAsync
    (:OnReady
      (if (or (Class|AthenaFunctionLibrary|RowtoBool (Variables|Default|GetRow_SpawnPhysicsVersion)) (Variables|Default|GetDebugAlwaysSpawnPhysicsVersion))
        (Variables|SpawnerVars|SetClassToSpawn (Variables|Default|GetPhysicsVersion))
        (else
          (Variables|SpawnerVars|SetClassToSpawn (Variables|Default|GetNonPhysicsVersion))
          (Variables|SpawnerVars|SetUsePrevisMeshScale false) ; Don't use the previs mesh scale if the non-physics version is spawned. As the none physics version is scaled differently
        ))
      (|CallParentFunction))))

(fn ConstructionScript
  (|CallParentFunction))
