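; Projectile Blueprint for TnTina fireballs that layers impact audio on top of its parent's behaviour.
; On stop, it fades out the looping audio tell and plays a flare gun ember impact sound at the impact point, unless the hit actor is a water body.
; Entering water instead plays a water impact sound at the projectile's location, guarded to fire only once.
; BeginPlay and hit events only forward to the parent.
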(event EventBeginPlay
  (|CallParentFunction))

(event Collision|EventHit
  (params (MyComp :type Primitive%20Component%20Object%20Reference) (Other :type Actor%20Object%20Reference) (OtherComp :type Primitive%20Component%20Object%20Reference) (bSelfMoved :type Boolean) (HitLocation :type Vector :default "0, 0, 0") (HitNormal :type Vector :default "0, 0, 0") (NormalImpulse :type Vector :default "0, 0, 0") (Hit :type Hit%20Result%20Structure%20%28by%20ref%29))
  (|CallParentFunction MyComp Other OtherComp bSelfMoved HitLocation HitNormal NormalImpulse Hit))

(event EventOnStop
  (params (Hit :type Hit%20Result%20Structure%20%28by%20ref%29))
  (bind (_bblockinghit _binitialoverlap _time _distance _location _impactpoint _normal _impactnormal _physmat _hitactor) (Collision|BreakHitResult Hit))
  (|CallParentFunction Hit)
  (Audio|Components|Audio|FadeOut (Variables|Default|GetLoopingAudioTell) 0.25 0.0 "Linear" :!class /Script/Engine.AudioComponent)
  (Utilities|Casting|CastToWaterBody _hitactor
    (:CastFailed
      (CallFunction|PlayImpactSound))))

(event Custom|Play%20Impact%20Sound
  (Audio|PlaySoundatLocation "/Game/Sounds/Fort_Weapons/pistol/FlareGun/FlareGun_EmberImpact_Cue.FlareGun_EmberImpact_Cue" (.ImpactPoint (Collision|BreakHitResult (stale EventOnStop Hit)))))

(event OnEnterWaterEvent(WaterInteractionComponent)
  (params (WaterBody :type Fort%20Water%20Body%20Actor%20Object%20Reference) (WaterInteractionComponent :type Fort%20Water%20Interaction%20Component%20Object%20Reference) (bIsFirstBody :type Boolean))
  ;[1 1 1] On water impact
  (Utilities|FlowControl|DoOnce false
    (:Completed
      (Audio|PlaySoundatLocation "/Game/Sounds/Fort_Weapons/pistol/FlareGun/FlareGun_EmberImpact_Water_Cue.FlareGun_EmberImpact_Water_Cue" (Transformation|GetActorLocation)))))

(fn ConstructionScript)
