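; Gameplay ability for an explosive held-object BGA that, on a gameplay event, damages and knocks back actors found by target selection.
; Each target receives the explosion effect, credited to the event's instigator.
; Non-immune pawns outside vehicles launch away from the blast; vehicles get an angle-clamped impulse along either velocity or blast direction, depending on speed.
; The physics-building impulse branch multiplies a zero vector, so imparts no force.
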(event Ability|EventOnEndAbility
  (params (bWasCancelled :type Boolean)))

(event Custom|ApplyDamage
  (params (Actor :type Actor%20Object%20Reference))
  (bind _returnvalue (Ability|GetActivatingAbilityComponent))
  ;[1 1 1] Apply Damage
  (bind _asfort_player_controller (Utilities|Casting|CastToFortPlayerController (Variables|Default|GetInstigator))
    (:then
      (Variables|Default|SetInstigatorAbility (Ability|GetFortAbilitySystemComponent (Pawn|GetControlledPawn _asfort_player_controller))))
    (:CastFailed
      (Variables|Default|SetInstigatorAbility _returnvalue)))
  (GameplayEffects|FortApplyGameplayEffectToTarget (Variables|Default|GetInstigatorAbility) (Variables|Default|GetGE_Explode) (Ability|GetFortAbilitySystemComponent Actor) (Variables|Default|GetInstigator) (Variables|Default|GetInstigator) (Components|GetOwner _returnvalue) 0.0 (CallFunction|MakeEffectContext)))

(event Ability|EventActivateAbilityFromEvent
  (params (EventData :type Gameplay%20Event%20Data%20Structure%20%28by%20ref%29))
  (bind (_eventtag _instigator) (Utilities|Struct|BreakGameplayEventData EventData))
  ;[1 1 1] Check for pawns in LOS of the explosion
  (Variables|Default|SetInstigator _instigator)
  (bind (_asynctaskproxy _targetdata) (Ability|Tasks|DoTargetSelection "None" "(TagName=\"Targeting.ApplicationTag.0\")")
    (:Targeted
      (for _array_element (Ability|TargetData|GetAllActorsfromTargetData _targetdata)
        (Variables|Default|SetCurrentPlayerToLaunch _array_element)
        (Utilities|FlowControl|Sequence
          (:then_0
            (CallFunction|ApplyDamage _array_element))
          (:then_1
            ;[1 1 1] Launch Player
            (bind _asfort_pawn (Utilities|Casting|CastToFortPawn _array_element)
              (:then
                (if (not (or (GameplayTags|HasMatchingGameplayTag _asfort_pawn "(TagName=\"Gameplay.Status.KnockbackImmunity\")") (GameplayTags|HasMatchingGameplayTag _asfort_pawn "(TagName=\"Gameplay.Action.Player.AthenaInVehicle\")")))
                  (Character|LaunchCharacter _asfort_pawn (CallFunction|CalculateLaunchVelocity (Transformation|GetActorLocation _array_element) (Transformation|GetActorLocation (Components|GetOwner (Ability|GetActivatingAbilityComponent)))) true true)))
              (:CastFailed
                ;[1 1 1] LaunchVehicle
                (bind _asfort_vehicle_interface (Utilities|Casting|CastToFortVehicleInterface _array_element)
                  (:then
                    (bind _returnvalue (Vehicle|GetVehicleActor _asfort_vehicle_interface :!class /Script/FortniteGame.FortVehicleInterface))
                    (Audio|BroadcastAbilityImpactatLocation 0 _returnvalue (select (>= (Math|Vector|VectorLength (Transformation|GetVelocity _returnvalue)) 5e+02) (Math|Vector|MakeVector (.x (* (Utilities|ClamptoMinAngleBetween (Math|Vector|Normalize (- (+ (Transformation|GetActorLocation _returnvalue) (Transformation|GetVelocity _returnvalue)) (Transformation|GetActorLocation _returnvalue))) (Math|Vector|ProjectVectorontoPlane (Math|Vector|Normalize (- (+ (Transformation|GetActorLocation _returnvalue) (Transformation|GetVelocity _returnvalue)) (Transformation|GetActorLocation _returnvalue))) "0, 0,1.000000") 3e+01 false) (ScalableFloat|GetValueatLevel (Variables|Default|GetVehicleImpulseStrength)))) (.y (* (Utilities|ClamptoMinAngleBetween (Math|Vector|Normalize (- (+ (Transformation|GetActorLocation _returnvalue) (Transformation|GetVelocity _returnvalue)) (Transformation|GetActorLocation _returnvalue))) (Math|Vector|ProjectVectorontoPlane (Math|Vector|Normalize (- (+ (Transformation|GetActorLocation _returnvalue) (Transformation|GetVelocity _returnvalue)) (Transformation|GetActorLocation _returnvalue))) "0, 0,1.000000") 3e+01 false) (ScalableFloat|GetValueatLevel (Variables|Default|GetVehicleImpulseStrength)))) (ScalableFloat|GetValueatLevel (Variables|Default|GetVehicleImpulseStrengthZ))) (Math|Vector|MakeVector (.x (* (Utilities|ClamptoMinAngleBetween (Math|Vector|Normalize (- (Transformation|GetActorLocation _array_element) (Transformation|GetActorLocation (Components|GetOwner (Ability|GetActivatingAbilityComponent))))) (Math|Vector|ProjectVectorontoPlane (Math|Vector|Normalize (- (Transformation|GetActorLocation _array_element) (Transformation|GetActorLocation (Components|GetOwner (Ability|GetActivatingAbilityComponent))))) "0, 0,1.000000") 3e+01 false) (ScalableFloat|GetValueatLevel (Variables|Default|GetVehicleImpulseStrength)))) (.y (* (Utilities|ClamptoMinAngleBetween (Math|Vector|Normalize (- (Transformation|GetActorLocation _array_element) (Transformation|GetActorLocation (Components|GetOwner (Ability|GetActivatingAbilityComponent))))) (Math|Vector|ProjectVectorontoPlane (Math|Vector|Normalize (- (Transformation|GetActorLocation _array_element) (Transformation|GetActorLocation (Components|GetOwner (Ability|GetActivatingAbilityComponent))))) "0, 0,1.000000") 3e+01 false) (ScalableFloat|GetValueatLevel (Variables|Default|GetVehicleImpulseStrength)))) (ScalableFloat|GetValueatLevel (Variables|Default|GetVehicleImpulseStrengthZ)))) 777.0))
                  (:CastFailed
                    ;[1 1 1] Launch Physics Actors
                    (bind _returnvalue_2 (GameplayTags|HasMatchingGameplayTag(Message) _array_element "(TagName=\"Building.Type.Physics\")"))
                    (if _returnvalue_2
                      (if (Collision|GetActorEnableCollision _array_element)
                        (bind _asprimitive_component (Utilities|Casting|CastToPrimitiveComponent (Variables|/Script/Engine.Actor|GetRootComponent _array_element)))
                        (Physics|AddImpulseatLocation _asprimitive_component (* "0, 0, 0" (ScalableFloat|GetValueatLevel (Variables|Default|GetRow_BuildingActorKnockMult))) (Physics|GetCenterOfMass _asprimitive_component "None") "None")))))))))))) ; This handles the original explosion damage
)

(fn CalculateLaunchVelocity
  (params (PlayerPosition :type Vector) (MyPosition :type Vector))
  (return (* (+ (Utilities|ClamptoMinAngleBetween (Math|Vector|Normalize (- PlayerPosition MyPosition)) (Math|Vector|ProjectVectorontoPlane (Math|Vector|Normalize (- PlayerPosition MyPosition)) "0, 0,1.000000") 0 false) "0, 0,0.300000") (ScalableFloat|GetValueatLevel (Variables|Default|GetKnockbackStrength)))))

(fn MakeEffectContext
  ;[1 1 1] Make the effect context for the Gameplay effect from the avatar (in this case it is the explosive BGA)
  (return (GameplayEffects|MakeFortEffectContext (Ability|GetFortAbilitySystemComponent (Ability|GetAvatarActorfromActorInfo)) (Variables|Default|GetInstigator) (Variables|Default|GetInstigator) (Ability|GetAvatarActorfromActorInfo) 0)))
